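# Sound cue for dropping the paint grenade item, layering a generic consumable drop with a quieter bounce.
# A Mixer combines a modulated random pick of two generic drop waves with a random pick of three Easter launcher bounce waves at reduced volume.
# It uses weapon drop attenuation and the loot collect sound class, with a concurrency override allowing one instance and preventing new ones while it plays.
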SoundCue /Game/Sounds/Fort_Weapons/Launchers/Paint/PaintGrenade_Drop_Cue
root=Mixer_0
nodes [
 {id=Mixer_0, children [Modulator_0,Random_1], properties {inputVolume [1,0.3]}}
 {id=Modulator_0, children [Random_0]}
 {id=Random_0, children [WavePlayer_1,WavePlayer_2], properties {weights [1,1]}}
 {id=Random_1, children [WavePlayer_0,WavePlayer_3,WavePlayer_4], properties {weights [1,1,1]}}
 {id=WavePlayer_0, properties {soundWaveAssetPtr=/Game/Sounds/Fort_Weapons/Launchers/Easter/Launcher_Easter_BounceA_01}}
 {id=WavePlayer_1, properties {soundWaveAssetPtr=/Game/Athena/Sounds/Items/Item_Consumables_Generic_Drop_02}}
 {id=WavePlayer_2, properties {soundWaveAssetPtr=/Game/Athena/Sounds/Items/Item_Consumables_Generic_Drop_01}}
 {id=WavePlayer_3, properties {soundWaveAssetPtr=/Game/Sounds/Fort_Weapons/Launchers/Easter/Launcher_Easter_BounceA_02}}
 {id=WavePlayer_4, properties {soundWaveAssetPtr=/Game/Sounds/Fort_Weapons/Launchers/Easter/Launcher_Easter_BounceA_03}}
]
properties {
 attenuationSettings=&Fort_Weapon_Drop_Att
 bOverrideConcurrency=true
 maxDistance=3000
 soundClassObject=&Fort_Loot_Collect
 volumeMultiplier=0.6
 concurrencyOverrides {maxCount=1, resolutionRule=PreventNew}
}
