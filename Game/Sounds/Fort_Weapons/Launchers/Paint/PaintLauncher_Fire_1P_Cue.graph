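# Sound cue for the first-person firing sound of the paint launcher weapon.
# A Random node picks one of four equally weighted close-range fire waves, and a Modulator at the root varies it using default settings.
# The cue slightly raises pitch, lowers volume, and uses a first-person ranged-fire sound class, first-person projectile attenuation and first-person gunfire concurrency.
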SoundCue /Game/Sounds/Fort_Weapons/Launchers/Paint/PaintLauncher_Fire_1P_Cue
root=Modulator_4
nodes [
 {id=Modulator_4, children [Random_2]}
 {id=Random_2, children [WavePlayer_14, WavePlayer_16, WavePlayer_17, WavePlayer_18], properties {weights [1,1,1,1]}}
 {id=WavePlayer_14, properties {soundWaveAssetPtr=&/PaintLauncher_Fire_Close_01}}
 {id=WavePlayer_16, properties {soundWaveAssetPtr=&/PaintLauncher_Fire_Close_02}}
 {id=WavePlayer_17, properties {soundWaveAssetPtr=&/PaintLauncher_Fire_Close_03}}
 {id=WavePlayer_18, properties {soundWaveAssetPtr=&/PaintLauncher_Fire_Close_04}}
]
properties {
 attenuationSettings=&Fort_Weapon_Projectile_FirstPerson_Athena_Att
 maxDistance=2097152
 pitchMultiplier=1.1
 soundClassObject=&Fort_WeaponRanged_Fire_1P_PSM
 volumeMultiplier=0.7
 attenuationOverrides {distanceAlgorithm=NaturalSound, falloffDistance=5000, attenuationShapeExtents {x=1000}}
 concurrencySet [&Weapon_Gunfire_FirstPerson_Conc]
}
