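# Sound cue for the Paint Launcher's firing sound, layered by distance.
# A mixer blends close, medium and distant branches, each with its own attenuation and a sound class override, the latter two sharing the distant-fire class.
# Close and medium branches randomly pick from four waves each through modulators, while the distant branch has no wave connected.
# Local and global gunfire concurrency sets limit simultaneous playback.
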SoundCue /Game/Sounds/Fort_Weapons/Launchers/Paint/PaintLauncher_Fire_Cue
root=Mixer_17
nodes [
 {
  id=Attenuation_10
  children [Modulator_14]
  properties {attenuationSettings=&Fort_Weapon_Projectile_Medium_Mono_Athena_Att}
 }
 {
  id=Attenuation_34
  children [Modulator_1]
  properties {attenuationSettings=&Fort_Weapon_Projectile_Distant_Mono_Athena_Att}
 }
 {id=Attenuation_9, children [Modulator_15], properties {attenuationSettings=&Fort_Weapon_Projectile_Close_Athena_Att}}
 {id=Mixer_17, children [SoundClass_7, SoundClass_6, SoundClass_35], properties {inputVolume [1.1,0.6,0.8]}}
 {id=Modulator_1, children [null], properties {volumeMax=1, volumeMin=1}}
 {id=Modulator_14, children [Random_1], properties {pitchMax=1.1, volumeMax=0.9, volumeMin=0.8}}
 {id=Modulator_15, children [Random_0]}
 {id=Random_0, children [WavePlayer_0,WavePlayer_1,WavePlayer_3,WavePlayer_4], properties {weights [1,1,1,1]}}
 {id=Random_1, children [WavePlayer_5, WavePlayer_10, WavePlayer_11, WavePlayer_12], properties {weights [1,1,1,1]}}
 {
  id=SoundClass_35
  children [Attenuation_34]
  properties {soundClassOverride=&Fort_WeaponRanged_Fire_Distant}
 }
 {
  id=SoundClass_6
  children [Attenuation_10]
  properties {soundClassOverride=&Fort_WeaponRanged_Fire_Distant}
 }
 {
  id=SoundClass_7
  children [Attenuation_9]
  properties {soundClassOverride=&Fort_WeaponRanged_Fire_3P_PSM}
 }
 {id=WavePlayer_0, properties {soundWaveAssetPtr=&/PaintLauncher_Fire_Close_01}}
 {id=WavePlayer_1, properties {soundWaveAssetPtr=&/PaintLauncher_Fire_Close_02}}
 {id=WavePlayer_10, properties {soundWaveAssetPtr=&/PaintLauncher_Fire_Medium_02}}
 {id=WavePlayer_11, properties {soundWaveAssetPtr=&/PaintLauncher_Fire_Medium_03}}
 {id=WavePlayer_12, properties {soundWaveAssetPtr=&/PaintLauncher_Fire_Medium_04}}
 {id=WavePlayer_3, properties {soundWaveAssetPtr=&/PaintLauncher_Fire_Close_03}}
 {id=WavePlayer_4, properties {soundWaveAssetPtr=&/PaintLauncher_Fire_Close_04}}
 {id=WavePlayer_5, properties {soundWaveAssetPtr=&/PaintLauncher_Fire_Medium_01}}
]
properties {
 maxDistance=27500
 pitchMultiplier=1.1
 volumeMultiplier=0.6
 attenuationOverrides {falloffDistance=5000, attenuationShapeExtents {x=1000}}
 concurrencyOverrides {bLimitToOwner=true, resolutionRule=StopOldest}
 concurrencySet [&Weapon_Gunfire_Local_Conc, &Weapon_Gunfire_Global_Conc]
}
