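# Sound cue for the paint launcher's impact sound, routed to the impacts sound class.
# A random node picks one of three equally weighted impact recordings, and a modulator varies its pitch slightly.
# Overridden attenuation uses logarithmic falloff with a low-pass filter that muffles distant hits.
# Concurrency values specify a single instance with new plays prevented, though no concurrency override flag is shown.
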SoundCue /Game/Sounds/Fort_Weapons/Launchers/Paint/PaintLauncher_Impact_Cue
root=Modulator_0
nodes [
 {id=Modulator_0, children [Random_0], properties {pitchMax=1.15, pitchMin=0.9}}
 {id=Random_0, children [WavePlayer_0,WavePlayer_1,WavePlayer_6], properties {weights [1,1,1]}}
 {id=WavePlayer_0, properties {soundWaveAssetPtr=&/PaintLauncher_Impact_01}}
 {id=WavePlayer_1, properties {soundWaveAssetPtr=&/PaintLauncher_Impact_02}}
 {id=WavePlayer_6, properties {soundWaveAssetPtr=&/PaintLauncher_Impact_03}}
]
properties {
 bOverrideAttenuation=true
 maxDistance=4000
 soundClassObject=&Fort_Impacts
 volumeMultiplier=1.04
 attenuationOverrides {
  bAttenuateWithLPF=true
  distanceAlgorithm=Logarithmic
  falloffDistance=3000
  lPFFrequencyAtMax=1000
  lPFRadiusMax=1500
  lPFRadiusMin=1000
  stereoSpread=800
  attenuationShapeExtents {x=1000}
 }
 concurrencyOverrides {maxCount=1, resolutionRule=PreventNew}
}
